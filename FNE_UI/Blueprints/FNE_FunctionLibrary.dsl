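; Blueprint function library of UI helpers for FNE_UI, covering dynamic material visual states and UMG layout.
; Transition functions set Trans_* scalar parameters on a MID for enabled, disabled, selected and deselected states, and CopyMIDParameterOverrides carries parameters to another MID to keep its look.
; Layout helpers cast a widget's slot to Grid or Overlay slot to set placement properties.
; Small utilities remove a widget from its parent and lerp 2D vectors.
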(fn CopyMIDParameterOverrides
  (params (Target%20MID :type Material%20Instance%20Dynamic%20Object%20Reference) (Source%20MID :type Material%20Instance%20Dynamic%20Object%20Reference))
  (if (Math|Boolean|ANDBoolean (Utilities|IsValid Target%20MID :!class /Script/Engine.KismetSystemLibrary) (Utilities|IsValid Source%20MID :!class /Script/Engine.KismetSystemLibrary) (!= Target%20MID Source%20MID))
    ;[0.06597 0.06597 0.06597 1 12] Passing internal MID parameter values to new MID to keep current visual state.
    (Rendering|Material|CopyParameterOverrides Target%20MID Source%20MID)))

(fn FNELerpVector2D
  (params (A_X :type Float%20%28double-precision%29) (A_Y :type Float%20%28double-precision%29) (B_X :type Float%20%28double-precision%29) (B_Y :type Float%20%28double-precision%29) (Alpha :type Float%20%28double-precision%29))
  (bind (_returnvalue_x _returnvalue_y) (Math|Vector|BreakVector (Math|Vector|Lerp(Vector) (Math|Vector|MakeVector A_X A_Y) (Math|Vector|MakeVector B_X B_Y) Alpha)))
  (return _returnvalue_x _returnvalue_y))

(fn InstantTransitionToDisabled
  (params (MID :type Material%20Instance%20Dynamic%20Object%20Reference) (IsSelected :type Boolean))
  (if IsSelected
    (Rendering|Material|SetScalarParameterValue MID "Trans_SelectedDisabled" 1.0 :!class /Script/Engine.MaterialInstanceDynamic)
    (else
      (Rendering|Material|SetScalarParameterValue MID "Trans_Disabled" 1.0 :!class /Script/Engine.MaterialInstanceDynamic))))

(fn InstantTransitionToEnabled
  (params (MID :type Material%20Instance%20Dynamic%20Object%20Reference) (IsSelected :type Boolean))
  (if (not IsSelected)
    (Rendering|Material|SetScalarParameterValue MID "Trans_Hovered" 0.0 :!class /Script/Engine.MaterialInstanceDynamic)
    (Rendering|Material|SetScalarParameterValue MID "Trans_Focused" 0.0 :!class /Script/Engine.MaterialInstanceDynamic)
    (Rendering|Material|SetScalarParameterValue MID "Trans_Pressed" 0.0 :!class /Script/Engine.MaterialInstanceDynamic)
    (Rendering|Material|SetScalarParameterValue MID "Trans_Disabled" 0.0 :!class /Script/Engine.MaterialInstanceDynamic)
    (Rendering|Material|SetScalarParameterValue MID "Trans_SelectedIdle" 0.0 :!class /Script/Engine.MaterialInstanceDynamic))
  (Rendering|Material|SetScalarParameterValue MID "Trans_SelectedHovered" 0.0 :!class /Script/Engine.MaterialInstanceDynamic)
  (Rendering|Material|SetScalarParameterValue MID "Trans_SelectedFocused" 0.0 :!class /Script/Engine.MaterialInstanceDynamic)
  (Rendering|Material|SetScalarParameterValue MID "Trans_SelectedPressed" 0.0 :!class /Script/Engine.MaterialInstanceDynamic)
  (Rendering|Material|SetScalarParameterValue MID "Trans_SelectedDisabled" 0.0 :!class /Script/Engine.MaterialInstanceDynamic))

(fn InstantTransitionToSelected
  (params (MID :type Material%20Instance%20Dynamic%20Object%20Reference))
  (Rendering|Material|SetScalarParameterValue MID "Trans_SelectedIdle" 1.0 :!class /Script/Engine.MaterialInstanceDynamic)
  (Rendering|Material|SetScalarParameterValue MID "Trans_SelectedHovered" 0.0 :!class /Script/Engine.MaterialInstanceDynamic)
  (Rendering|Material|SetScalarParameterValue MID "Trans_SelectedFocused" 0.0 :!class /Script/Engine.MaterialInstanceDynamic)
  (Rendering|Material|SetScalarParameterValue MID "Trans_SelectedPressed" 0.0 :!class /Script/Engine.MaterialInstanceDynamic)
  (Rendering|Material|SetScalarParameterValue MID "Trans_SelectedDisabled" 0.0 :!class /Script/Engine.MaterialInstanceDynamic))

(fn RemoveWidget
  (params (Widget :type Widget%20Object%20Reference))
  (if (Utilities|IsValid Widget :!class /Script/Engine.KismetSystemLibrary)
    (Widget|Panel|RemoveChild (Widget|GetParent Widget) Widget)))

(fn SetGridSlotAlignment
  (params (InGridSlot :type Widget%20Object%20Reference) (InHorizontal :type EHorizontalAlignment%20Enum :default "HAlign_Fill") (InVertical :type EVerticalAlignment%20Enum :default "VAlign_Fill"))
  (bind _asgrid_slot (Utilities|Casting|CastToGridSlot (Variables|/Script/UMG.Widget|GetSlot InGridSlot)))
  (Layout|GridSlot|SetHorizontalAlignment _asgrid_slot InHorizontal)
  (Layout|GridSlot|SetVerticalAlignment _asgrid_slot InVertical))

(fn SetGridSlotColumnAndRow
  (params (Widget :type Widget%20Object%20Reference) (Row :type Integer) (Row%20Span :type Integer) (Column :type Integer) (Column%20Span :type Integer))
  (bind _asgrid_slot (Utilities|Casting|CastToGridSlot (Variables|/Script/UMG.Widget|GetSlot Widget)))
  (Layout|GridSlot|SetRow _asgrid_slot Row)
  (Layout|GridSlot|SetRowSpan _asgrid_slot Row%20Span)
  (Layout|GridSlot|SetColumn _asgrid_slot Column)
  (Layout|GridSlot|SetColumnSpan _asgrid_slot Column%20Span))

(fn SetGridSlotLayer
  (params (Widget :type Widget%20Object%20Reference) (Layer :type Integer))
  (bind _asgrid_slot (Utilities|Casting|CastToGridSlot (Variables|/Script/UMG.Widget|GetSlot Widget)))
  (Layout|GridSlot|SetLayer _asgrid_slot Layer))

(fn SetGridSlotNudge
  (params (Widget :type Widget%20Object%20Reference) (Nudge :type Vector%202D%20Structure))
  (bind _asgrid_slot (Utilities|Casting|CastToGridSlot (Variables|/Script/UMG.Widget|GetSlot Widget)))
  (Layout|GridSlot|SetNudge _asgrid_slot Nudge))

(fn SetGridSlotPadding
  (params (Widget :type Widget%20Object%20Reference) (Padding :type Margin%20Structure))
  (bind _asgrid_slot (Utilities|Casting|CastToGridSlot (Variables|/Script/UMG.Widget|GetSlot Widget)))
  (Layout|BorderSlot|SetPadding _asgrid_slot Padding :!class /Script/UMG.GridSlot))

(fn SetOverlaySlotAlignment
  (params (Widget :type Widget%20Object%20Reference) (Horizontal :type EHorizontalAlignment%20Enum :default "HAlign_Fill") (Vertical :type EVerticalAlignment%20Enum :default "VAlign_Fill"))
  (bind _asoverlay_slot (Utilities|Casting|CastToOverlaySlot (Variables|/Script/UMG.Widget|GetSlot Widget)))
  (Layout|OverlaySlot|SetHorizontalAlignment _asoverlay_slot Horizontal)
  (Layout|OverlaySlot|SetVerticalAlignment _asoverlay_slot Vertical))

(fn TransitionToDeselectedFocused
  (params (MID :type Material%20Instance%20Dynamic%20Object%20Reference))
  (Rendering|Material|SetScalarParameterValue MID "Trans_SelectedIdle" 0.0 :!class /Script/Engine.MaterialInstanceDynamic)
  (Rendering|Material|SetScalarParameterValue MID "Trans_SelectedHovered" 0.0 :!class /Script/Engine.MaterialInstanceDynamic)
  (Rendering|Material|SetScalarParameterValue MID "Trans_SelectedPressed" 0.0 :!class /Script/Engine.MaterialInstanceDynamic)
  (Rendering|Material|SetScalarParameterValue MID "Trans_SelectedDisabled" 0.0 :!class /Script/Engine.MaterialInstanceDynamic)
  (Rendering|Material|SetScalarParameterValue MID "Trans_Pressed" 0.0 :!class /Script/Engine.MaterialInstanceDynamic)
  (Rendering|Material|SetScalarParameterValue MID "Trans_Disabled" 0.0 :!class /Script/Engine.MaterialInstanceDynamic))

(fn TransitionToDeselectedIdle
  (params (MID :type Material%20Instance%20Dynamic%20Object%20Reference))
  (Rendering|Material|SetScalarParameterValue MID "Trans_SelectedHovered" 0.0 :!class /Script/Engine.MaterialInstanceDynamic)
  (Rendering|Material|SetScalarParameterValue MID "Trans_SelectedFocused" 0.0 :!class /Script/Engine.MaterialInstanceDynamic)
  (Rendering|Material|SetScalarParameterValue MID "Trans_SelectedPressed" 0.0 :!class /Script/Engine.MaterialInstanceDynamic)
  (Rendering|Material|SetScalarParameterValue MID "Trans_SelectedDisabled" 0.0 :!class /Script/Engine.MaterialInstanceDynamic)
  (Rendering|Material|SetScalarParameterValue MID "Trans_Hovered" 0.0 :!class /Script/Engine.MaterialInstanceDynamic)
  (Rendering|Material|SetScalarParameterValue MID "Trans_Focused" 0.0 :!class /Script/Engine.MaterialInstanceDynamic)
  (Rendering|Material|SetScalarParameterValue MID "Trans_Pressed" 0.0 :!class /Script/Engine.MaterialInstanceDynamic)
  (Rendering|Material|SetScalarParameterValue MID "Trans_Disabled" 0.0 :!class /Script/Engine.MaterialInstanceDynamic))
